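# Niagara idle effect for the Embers glider, dressing its wheels, frame, handlebar and eye in fire.
# Single-particle mesh emitters follow bones on the user-supplied skeletal mesh, with some wheel flames spun by a system-level integrator.
# Flipbook sprite emitters spawn on cylinders around the wheel bones, using a local module to track bone movement as flames rise.
# Four offset lights flicker through looping color and intensity curves.
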NiagaraSystem /BRCosmetics/Gadgets/Assets/VinderTech_GliderChute/Glider_Embers_Male/FX/NS_Embers_Glider_Idle
dependencies {
 usedDataInterfaces [&NiagaraDataInterfaceCurve, &NiagaraDataInterfaceVectorCurve, &NiagaraDataInterfaceColorCurve]
 usedDynamicInputs []
 usedModules [&SystemState, &FN_Integrator]
}
emitters {
 Eyefire {
  emitterUpdateScript [
   &EmitterState(
    Inactive Response=::NewEnumerator1
    Life Cycle Mode=::NewEnumerator1
    Loop Behavior=::NewEnumerator1
    Loop Duration Mode=::NewEnumerator1
    UseLoopDelay=false
   )
   &SpawnBurst_Instantaneous(Spawn Count=1, Use Spawn Probability=false)
  ]
  particleSpawnScript [/Niagara/Modules/Spawn/Initialization/InitializeParticle(
    Lifetime=5
    Module.Write Color=true
    Module.Write Lifetime=true
    Module.Write Mass=true
    Module.Write Material Random=true
    Module.Write Position=true
    Module.Write Scale=false
    Module.Write SpriteRotation=false
    Module.Write SpriteSize=true
    Sprite Size=<5 5>
   )]
  particleUpdateScript [
   &ParticleState(Kill Particles When Lifetime Has Elapsed=false, Let Infinitely Lived Particles Die When Emitter Deactivates=true)
   &SkeletalMeshLocation(
    Apply Sampled Orientation to Mesh Particles=true
    Bone / Socket Index=3
    Bone Sampling Mode=::NewEnumerator4
    Inherit Velocity=false
    Interpolate Mesh Sampling=false
    MaskSpawnGroup=false
    Sampled Position Offset=<17 0 -3>
    Sampled Position Offset Coordinate Space:ENiagaraMeshTransforms=::NewEnumerator3
    Skeletal Mesh=$User.SkeletalMesh
   ) {module=SkeletalMeshLocation001}
  ]
  rendererClasses [&NiagaraMeshRendererProperties]
  renderers [{rendererClass=&NiagaraMeshRendererProperties, rendererIndex=0}]
 }
 Frame_FIre {
  emitterUpdateScript [
   &EmitterState(
    Inactive Response=::NewEnumerator1
    Life Cycle Mode=::NewEnumerator1
    Loop Behavior=::NewEnumerator1
    Loop Delay=1.2
    Loop Duration Mode=::NewEnumerator1
    UseLoopDelay=false
   )
   &SpawnBurst_Instantaneous(Spawn Count=1, Use Spawn Probability=false)
  ]
  particleSpawnScript [&InitializeParticle(
    Module.Write Color=true
    Module.Write Lifetime=true
    Module.Write Mass=true
    Module.Write Material Random=true
    Module.Write Position=true
    Module.Write Scale=false
    Module.Write SpriteRotation=false
    Module.Write SpriteSize=true
    Sprite Size=<5 5>
   )]
  particleUpdateScript [
   &ParticleState(Kill Particles When Lifetime Has Elapsed=false, Let Infinitely Lived Particles Die When Emitter Deactivates=true)
   &SkeletalMeshLocation(
    Apply Sampled Orientation to Mesh Particles=true
    Bone / Socket Index=2
    Bone Sampling Mode=::NewEnumerator4
    Inherit Velocity=false
    Interpolate Mesh Sampling=false
    MaskSpawnGroup=false
    Sampled Position Offset=<-16 0 30>
    Sampled Position Offset Coordinate Space:ENiagaraMeshTransforms=::NewEnumerator3
    Skeletal Mesh=$User.SkeletalMesh
   ) {module=SkeletalMeshLocation001}
   /SetVariables(Particles.MeshOrientation=&MultiplyQuaternion(Quaternion A=$Output.SkeletalMeshLocation001.SampledRotation, Quaternion B=<0 0 -1 1>)) {isSetParameters=true}
  ]
  rendererClasses [&NiagaraMeshRendererProperties]
  renderers [{rendererClass=&NiagaraMeshRendererProperties, rendererIndex=0}]
 }
 Handlebar_Fire {
  emitterUpdateScript [
   &EmitterState(
    Inactive Response=::NewEnumerator1
    Life Cycle Mode=::NewEnumerator1
    Loop Behavior=::NewEnumerator1
    Loop Duration Mode=::NewEnumerator1
    UseLoopDelay=false
   )
   &SpawnBurst_Instantaneous(Spawn Count=1, Use Spawn Probability=false)
  ]
  particleSpawnScript [&InitializeParticle(
    Lifetime=5
    Module.Write Color=true
    Module.Write Lifetime=true
    Module.Write Mass=true
    Module.Write Material Random=true
    Module.Write Position=true
    Module.Write Scale=false
    Module.Write SpriteRotation=false
    Module.Write SpriteSize=true
    Sprite Size=<5 5>
   )]
  particleUpdateScript [
   &ParticleState(Kill Particles When Lifetime Has Elapsed=false, Let Infinitely Lived Particles Die When Emitter Deactivates=true)
   &SkeletalMeshLocation(
    Apply Sampled Orientation to Mesh Particles=true
    Bone / Socket Index=3
    Bone Sampling Mode=::NewEnumerator4
    Inherit Velocity=false
    Interpolate Mesh Sampling=false
    MaskSpawnGroup=false
    Sampled Position Offset=<10 0 -24>
    Sampled Position Offset Coordinate Space:ENiagaraMeshTransforms=::NewEnumerator3
    Skeletal Mesh=$User.SkeletalMesh
   ) {module=SkeletalMeshLocation001}
  ]
  rendererClasses [&NiagaraMeshRendererProperties]
  renderers [{rendererClass=&NiagaraMeshRendererProperties, rendererIndex=0}]
 }
 Light {
  emitterUpdateScript [
   &EmitterState(Inactive Response=::NewEnumerator1, Life Cycle Mode=::NewEnumerator1, Loop Duration=2, UseLoopDelay=true)
   &SpawnBurst_Instantaneous(Spawn Count=4, Use Spawn Probability=false)
  ]
  particleSpawnScript [
   &InitializeParticle(
    Color=<rgba 4 1.3574 0.0361 1>
    Lifetime=2
    Module.Write Color=true
    Module.Write Lifetime=true
    Module.Write Mass=false
    Module.Write Material Random=true
    Module.Write Position=false
    Module.Write Scale=false
    Module.Write SpriteRotation=false
    Module.Write SpriteSize=true
    Position=&SimulationPosition()
   )
   &FN_SetSpawnGroup_v2(Number of Spawn Groups=4)
   &DynamicMaterialParameters(
    Index 0 Param 1=&UniformRangedFloat(Minimum=0.7, Maximum=0.9)
    Index 0 Param 2=&UniformRangedFloat(Minimum=0.05, Maximum=0.15)
    Index 0 Param 3=1
    Module.Param0WriteEnabled=true
    Module.Param1WriteEnabled=true
    Module.Param2WriteEnabled=false
    Module.Param3WriteEnabled=false
   ) {module=DynamicMaterialParameters002}
   &SystemLocation(
    MaskSpawnGroup=false
    Offset=&MakeCustomVectorFromBool(
     Bool=&SetBoolByIntComparison(A=&SpawnGroup(), B=0, Comparison Type=::NewEnumerator4)
     True Vector=<0 75 50>
     False Vector=&MakeCustomVectorFromBool(
      Bool=&SetBoolByIntComparison(A=&SpawnGroup(), B=1, Comparison Type=::NewEnumerator4)
      True Vector=<0 -150 50>
      False Vector=&MakeCustomVectorFromBool(
       Bool=&SetBoolByIntComparison(A=&SpawnGroup(), B=2, Comparison Type=::NewEnumerator4)
       True Vector=<75 -50 50>
       False Vector=<-75 -50 50>
      )
     )
    )
   )
  ]
  particleUpdateScript [
   &ParticleState()
   &ScaleColor(
    Scale RGB=&Multiply_Vector(
     A=&MakeVectorFromLinearColorRGB(Color=&LerpLinearColors(
       StartColor=<rgba 2 1.5516 0.3059 1>
       EndColor=<rgba 2 0.4802 0 1>
       LerpFactor=&FloatFromCurve(
        FloatCurve=DataInterface() {propertyValues {Curve {keys [<at 0 1>,<at 0.1069 0.0895>,<at 0.3206 0.7423>,<at 0.4907 0.1444>,<at 0.6992 0.8948>,<at 0.9061 0.1261>,<at 1 1>]}}}
        CurveIndex=$Emitter.NormalizedLoopAge
        Scale Curve=0.9
       )
      ))
     B=&VectorFromFloat(Value=&FloatFromCurve(
       FloatCurve=DataInterface() {propertyValues {Curve {keys [<at 0 1>,<at 0.0584 1.1216>,<at 0.0797 0.6691>,<at 0.1464 0.526>,<at 0.222 1.0327>,<at 0.2705 0.6844>,<at 0.3166 0.8798>,<at 0.3641 0.5477>,<at 0.4407 0.6405>,<at 0.4815 1.1358>,<at 0.5388 1.0006>,<at 0.58 0.5314>,<at 0.6479 0.8407>,<at 0.6937 0.7459>,<at 0.737 1.0896>,<at 0.8125 0.5703>,<at 0.8909 0.8732>,<at 0.9449 0.6186>,<at 1 1>]}}}
       CurveIndex=$Emitter.NormalizedLoopAge
      ))
    )
    ScaleA=true
    ScaleRGB=true
   )
   &FN_Light_v2(Light Exponent=2, Light Radius=100, LightEnabled=true, Scale Color=1)
   &ScaleSpriteSize(Scale Factor=&Vector2DFromFloat(Value=&FloatFromCurve(
      FloatCurve=DataInterface() {propertyValues {Curve {keys [<at 0 1>,<at 0.0562 0.9129>,<at 0.0748 1.1321>,<at 0.1235 1.3668>,<at 0.2098 0.8951>,<at 0.2589 1.1544>,<at 0.2987 1.0516>,<at 0.3338 1.2874>,<at 0.3749 1.3981>,<at 0.4815 0.8928>,<at 0.5082 1.113>,<at 0.5446 1.2763>,<at 0.6237 1.1354>,<at 0.6838 1.0572>,<at 0.7109 0.8761>,<at 0.7888 0.7005>,<at 0.8531 1.2103>,<at 0.9284 1.1544>,<at 1 1>]}}}
      CurveIndex=$Emitter.NormalizedLoopAge
     )))
  ]
  rendererClasses [&NiagaraLightRendererProperties]
  renderers [{rendererClass=&NiagaraLightRendererProperties, rendererIndex=0}]
 }
 WheelFire_Back {
  emitterUpdateScript [
   &EmitterState(
    Inactive Response=::NewEnumerator1
    Life Cycle Mode=::NewEnumerator1
    Loop Behavior=::NewEnumerator1
    Loop Duration Mode=::NewEnumerator1
    UseLoopDelay=false
   )
   &SpawnBurst_Instantaneous(Spawn Count=1, Use Spawn Probability=false)
  ]
  particleSpawnScript [&InitializeParticle(
    Lifetime=5
    Module.Write Color=true
    Module.Write Lifetime=true
    Module.Write Mass=true
    Module.Write Material Random=true
    Module.Write Position=true
    Module.Write Scale=true
    Module.Write SpriteRotation=false
    Module.Write SpriteSize=false
    Sprite Size=<5 5>
   )]
  particleUpdateScript [
   &ParticleState(Kill Particles When Lifetime Has Elapsed=false, Let Infinitely Lived Particles Die When Emitter Deactivates=true)
   &SkeletalMeshLocation(
    Bone / Socket Index=1
    Bone Sampling Mode=::NewEnumerator4
    Inherit Velocity=false
    Interpolate Mesh Sampling=false
    MaskSpawnGroup=false
    Sampled Position Offset=<0 0 8>
    Skeletal Mesh=$User.SkeletalMesh
   ) {module=SkeletalMeshLocation001}
   /SetVariables(Particles.MeshOrientation=&MultiplyQuaternion(
     Quaternion A=$Output.SkeletalMeshLocation001.SampledRotation
     Quaternion B=&MakeQuatFromAxisAngle(Axis=<1 0 0>, AngleInDegrees=&Multiply_Float(A=$System.FN_Integrator002.Value, B=-60))
    )) {isSetParameters=true, module=SetVariables_#1}
   &CameraOffset(Camera Offset Amount=5)
  ]
  rendererClasses [&NiagaraMeshRendererProperties]
  renderers [{rendererClass=&NiagaraMeshRendererProperties, rendererIndex=0}]
 }
 WheelFire_Back_Bottom {
  emitterUpdateScript [
   &EmitterState(
    Inactive Response=::NewEnumerator1
    Life Cycle Mode=::NewEnumerator1
    Loop Behavior=::NewEnumerator1
    Loop Duration Mode=::NewEnumerator1
    UseLoopDelay=false
   )
   &SpawnBurst_Instantaneous(Spawn Count=1, Use Spawn Probability=false)
  ]
  particleSpawnScript [&InitializeParticle(
    Lifetime=5
    Module.Write Color=true
    Module.Write Lifetime=true
    Module.Write Mass=true
    Module.Write Material Random=true
    Module.Write Position=true
    Module.Write Scale=true
    Module.Write SpriteRotation=false
    Module.Write SpriteSize=false
    Sprite Size=<5 5>
   )]
  particleUpdateScript [
   &ParticleState(Kill Particles When Lifetime Has Elapsed=false, Let Infinitely Lived Particles Die When Emitter Deactivates=true)
   &SkeletalMeshLocation(
    Bone / Socket Index=1
    Bone Sampling Mode=::NewEnumerator4
    Inherit Velocity=false
    Interpolate Mesh Sampling=false
    MaskSpawnGroup=false
    Skeletal Mesh=$User.SkeletalMesh
   ) {module=SkeletalMeshLocation001}
   /SetVariables(Particles.MeshOrientation=&MultiplyQuaternion(
     Quaternion A=$Output.SkeletalMeshLocation001.SampledRotation
     Quaternion B=&MakeQuatFromAxisAngle(Axis=<1 0 0>, AngleInDegrees=&Multiply_Float(A=$System.FN_Integrator002.Value, B=-60))
    )) {isSetParameters=true, module=SetVariables_#1}
   &CameraOffset(Camera Offset Amount=2.5)
  ]
  rendererClasses [&NiagaraMeshRendererProperties]
  renderers [{rendererClass=&NiagaraMeshRendererProperties, rendererIndex=0}]
 }
 WheelFire_ForwardFacing {
  emitterUpdateScript [
   &EmitterState(
    Inactive Response=::NewEnumerator1
    Life Cycle Mode=::NewEnumerator1
    Loop Behavior=::NewEnumerator1
    Loop Duration Mode=::NewEnumerator1
    UseLoopDelay=false
   )
   &SpawnBurst_Instantaneous(Spawn Count=1, Use Spawn Probability=false)
  ]
  particleSpawnScript [&InitializeParticle(
    Lifetime=5
    Module.Write Color=true
    Module.Write Lifetime=true
    Module.Write Mass=true
    Module.Write Material Random=true
    Module.Write Position=true
    Module.Write Scale=true
    Module.Write SpriteRotation=false
    Module.Write SpriteSize=false
    Sprite Size=<5 5>
   )]
  particleUpdateScript [
   &ParticleState(Kill Particles When Lifetime Has Elapsed=false, Let Infinitely Lived Particles Die When Emitter Deactivates=true)
   &SkeletalMeshLocation(
    Apply Sampled Orientation to Mesh Particles=true
    Bone Sampling Mode=::NewEnumerator4
    Inherit Velocity=false
    Interpolate Mesh Sampling=false
    MaskSpawnGroup=false
    Skeletal Mesh=$User.SkeletalMesh
   ) {module=SkeletalMeshLocation001}
   &CameraOffset(Camera Offset Amount=7.5)
  ]
  rendererClasses [&NiagaraMeshRendererProperties]
  renderers [{rendererClass=&NiagaraMeshRendererProperties, rendererIndex=0}]
 }
 WheelFire_Front {
  emitterUpdateScript [
   &EmitterState(
    Inactive Response=::NewEnumerator1
    Life Cycle Mode=::NewEnumerator1
    Loop Behavior=::NewEnumerator1
    Loop Duration Mode=::NewEnumerator1
    UseLoopDelay=false
   )
   &SpawnBurst_Instantaneous(Spawn Count=1, Use Spawn Probability=false)
  ]
  particleSpawnScript [
   &InitializeParticle(
    Lifetime=5
    Module.Write Color=true
    Module.Write Lifetime=true
    Module.Write Mass=true
    Module.Write Material Random=true
    Module.Write Position=true
    Module.Write Scale=true
    Module.Write SpriteRotation=false
    Module.Write SpriteSize=false
    Sprite Size=<5 5>
   )
   /SetVariables(Particles.MeshScale:Vector3f=<30 30 30>) {isSetParameters=true}
  ]
  particleUpdateScript [
   &ParticleState(Kill Particles When Lifetime Has Elapsed=false, Let Infinitely Lived Particles Die When Emitter Deactivates=true)
   &SkeletalMeshLocation(
    Bone Sampling Mode=::NewEnumerator4
    Inherit Velocity=false
    Interpolate Mesh Sampling=false
    MaskSpawnGroup=false
    Skeletal Mesh=$User.SkeletalMesh
   ) {module=SkeletalMeshLocation001}
   /SetVariables(Particles.MeshOrientation=&MultiplyQuaternion(
     Quaternion A=$Output.SkeletalMeshLocation001.SampledRotation
     Quaternion B=&MakeQuatFromAxisAngle(Axis=<1 0 0>, AngleInDegrees=&Multiply_Float(A=$System.FN_Integrator002.Value, B=-60))
    )) {isSetParameters=true, module=SetVariables_#1}
  ]
  rendererClasses [&NiagaraMeshRendererProperties]
  renderers [{rendererClass=&NiagaraMeshRendererProperties, rendererIndex=0}]
 }
 WheelFire_Front_Bottom {
  emitterUpdateScript [
   &EmitterState(
    Inactive Response=::NewEnumerator1
    Life Cycle Mode=::NewEnumerator1
    Loop Behavior=::NewEnumerator1
    Loop Duration Mode=::NewEnumerator1
    UseLoopDelay=false
   )
   &SpawnBurst_Instantaneous(Spawn Count=1, Use Spawn Probability=false)
  ]
  particleSpawnScript [&InitializeParticle(
    Lifetime=5
    Module.Write Color=true
    Module.Write Lifetime=true
    Module.Write Mass=true
    Module.Write Material Random=true
    Module.Write Position=true
    Module.Write Scale=true
    Module.Write SpriteRotation=false
    Module.Write SpriteSize=false
    Sprite Size=<5 5>
   )]
  particleUpdateScript [
   &ParticleState(Kill Particles When Lifetime Has Elapsed=false, Let Infinitely Lived Particles Die When Emitter Deactivates=true)
   &SkeletalMeshLocation(
    Bone Sampling Mode=::NewEnumerator4
    Inherit Velocity=false
    Interpolate Mesh Sampling=false
    MaskSpawnGroup=false
    Skeletal Mesh=$User.SkeletalMesh
   ) {module=SkeletalMeshLocation001}
   /SetVariables(Particles.MeshOrientation=&MultiplyQuaternion(
     Quaternion A=$Output.SkeletalMeshLocation001.SampledRotation
     Quaternion B=&MakeQuatFromAxisAngle(Axis=<1 0 0>, AngleInDegrees=&Multiply_Float(A=$System.FN_Integrator002.Value, B=-60))
    )) {isSetParameters=true, module=SetVariables_#1}
   &CameraOffset(Camera Offset Amount=7.5)
  ]
  rendererClasses [&NiagaraMeshRendererProperties]
  renderers [{rendererClass=&NiagaraMeshRendererProperties, rendererIndex=0}]
 }
 Wheel_Sprites_Back {
  emitterUpdateScript [
   &EmitterState()
   &SpawnRate(SpawnRate=30, Use Spawn Probability=false)
  ]
  particleSpawnScript [
   &InitializeParticle(
    Color=<rgba 5 3.0056 1.6206 0.75>
    Lifetime=&UniformRangedFloat(Minimum=0.48, Maximum=0.9)
    Module.Write Color=false
    Module.Write Lifetime=true
    Module.Write Mass=false
    Module.Write Material Random=true
    Module.Write Position=false
    Module.Write Scale=false
    Module.Write SpriteRotation=true
    Module.Write SpriteSize=true
    Position=&SimulationPosition()
    Sprite Rotation=&UniformRangedFloat(Minimum=-20, Maximum=20)
    Sprite Size=&MakeCustomVector2DFromBool(Bool=&RandomBool(), True Vector2D=<52 52>, False Vector2D=<-52 52>)
   )
   &FN_SetSpawnGroup_v2(Number of Spawn Groups=2)
   &SkeletalMeshLocation(
    Apply To Particle Position=false
    Bone / Socket Index=1
    Bone Sampling Mode=::NewEnumerator4
    Inherit Velocity=false
    Interpolate Mesh Sampling=false
    MaskSpawnGroup=false
    Skeletal Mesh=$User.SkeletalMesh
   )
   &CylinderLocation(
    Cylinder Height=10
    Cylinder Radius=20
    Hemisphere Y=true
    Lathe Profile=DataInterface() {propertyValues {Curve {keys [<at 0 0.451 RCIM_Cubic>, {arriveTangent=0.049, interpMode=RCIM_Cubic, leaveTangent=0.049, time=0.5, value=1}, <at 1 0.5 RCIM_Cubic>]}}}
    MaskSpawnGroup=false
    Module.Enable Lathe Profile=false
    Module.Override Randomness=false
    Module.Surface Only=true
    Non Uniform Scale:Vector3f=<1 1 -1>
    Offset=<0 0 -10>
    Orientation Axis:ENiagaraOrientationAxis=::XAxis
    Surface Only Band Thickness=0
    Use Endcaps In Surface Only Mode=false
   ) {module=CylinderLocation001}
   &DynamicMaterialParameters(
    Index 0 Param 2=&UniformRangedFloat(Minimum=0, Maximum=1)
    Index 0 Param 3=1
    Index 0 Param 4=&UniformRangedFloat(Minimum=1, Maximum=2)
    Module.Param0WriteEnabled=false
    Module.Param1WriteEnabled=true
    Module.Param2WriteEnabled=false
    Module.Param3WriteEnabled=true
   ) {module=DynamicMaterialParameters002}
   /SetVariables(Particles.CylVector:Vector3f=$Output.CylinderLocation001.CylinderVector) {isSetParameters=true, module=SetVariables_#2}
   /SetVariables(Particles.DestinationVector:Vector3f=&Add_Vector(A=$Output.SkeletalMeshLocation.SampledPosition, B=$Particles.CylVector)) {isSetParameters=true, module=SetVariables_#3}
  ]
  particleUpdateScript [
   &ParticleState()
   &SkeletalMeshLocation(
    Apply To Particle Position=false
    Bone / Socket Index=1
    Bone Sampling Mode=::NewEnumerator4
    Inherit Velocity=false
    Interpolate Mesh Sampling=false
    MaskSpawnGroup=false
    Skeletal Mesh=$User.SkeletalMesh
   ) {module=SkeletalMeshLocation001}
   &self:UseUpdatedSkelLocation(
    DestinationPosition:Vector3f=&Add_Vector(A=$Particles.CylVector, B=$Output.SkeletalMeshLocation001.SampledPosition)
    Normalized Influence:NiagaraFloat=&FloatFromCurve(FloatCurve=DataInterface() {propertyValues {Curve {keys [<at 0 1>,<at 1 0.91>]}}})
    Offset Position:Vector3f=&RandomRangeVector(
     Minimum=&VectorFromCurve(VectorCurve=DataInterface() {propertyValues {
        XCurve {keys [<at 0 -2>]}
        YCurve {keys [<at 0 -2>]}
        ZCurve {keys [<at 0 0 RCIM_Cubic>, {arriveTangent=29.403, interpMode=RCIM_Cubic, leaveTangent=29.4031, tangentMode=RCTM_User, time=1, value=22}]}
       }})
     Maximum=&VectorFromCurve(VectorCurve=DataInterface() {propertyValues {
        XCurve {keys [<at 0 3>]}
        YCurve {keys [<at 0 3>]}
        ZCurve {keys [<at 0 0 RCIM_Cubic>, {arriveTangent=29.403, interpMode=RCIM_Cubic, leaveTangent=29.4031, tangentMode=RCTM_User, time=1, value=22}]}
       }})
    )
    SourcePosition:Vector3f=$Particles.DestinationVector
   ) {module=UseUpdatedSkelLocation}
   &ScaleColor(
    Scale Mode=::NewEnumerator1
    Scale RGBA=&Vector4FromCurve(Vector4Curve=DataInterface() {propertyValues {
       AlphaCurve {keys [<at 0.0024 0>,<at 0.1271 1.2>,<at 0.75 4>]}
       BlueCurve {keys [<at 0.009 0.15>,<at 0.25 0>,<at 0.6 0>,<at 0.75 0>]}
       GreenCurve {keys [<at 0.009 0.5841>,<at 0.25 0.125>,<at 0.6 0.0833>,<at 0.75 0>]}
       RedCurve {keys [<at 0.009 1>,<at 0.25 1>,<at 0.6 1>,<at 0.75 1>]}
      }})
    ScaleRGBA=true
   ) {module=ScaleColor002}
   /SetVariables(Particles.SubImageIndex=&Add_Float(
     A=&RandomRangeFloat(Minimum=0, Maximum=20)
     B=&FloatFromCurve(FloatCurve=DataInterface() {propertyValues {Curve {keys [<at 0 0>,<at 1 24>]}}})
    )) {isSetParameters=true, module=SetVariables_#4}
   &ScaleSpriteSize(Scale Factor=&Vector2DFromFloat(Value=&FloatFromCurve(FloatCurve=DataInterface() {propertyValues {Curve {keys [
          {arriveTangent=7.5764, interpMode=RCIM_Cubic, leaveTangent=7.5765, tangentMode=RCTM_User, time=0, value=0}
          {arriveTangent=-0.0353, interpMode=RCIM_Cubic, leaveTangent=-0.0353, tangentMode=RCTM_User, time=0.2, value=1}
          <at 0.5752 1>
          <at 1 0 RCIM_Cubic>
         ]}}})))
   &DynamicMaterialParameters(
    Index 0 Param 1=&FloatFromCurve(FloatCurve=DataInterface() {propertyValues {Curve {keys [{arriveTangent=2.7551, interpMode=RCIM_Cubic, leaveTangent=2.7551, tangentMode=RCTM_User, time=0, value=0}, <at 0.5 1>]}}})
    Index 0 Param 2=&RandomRangeFloat(Minimum=0.33, Maximum=1)
    Index 0 Param 3=&FloatFromCurve(FloatCurve=DataInterface() {propertyValues {Curve {keys [<at 0 1>,<at 1 2>]}}})
    Index 0 Param 4=1
    Module.Param0WriteEnabled=true
    Module.Param1WriteEnabled=true
    Module.Param2WriteEnabled=true
    Module.Param3WriteEnabled=true
   ) {module=DynamicMaterialParameters001}
   &CameraOffset(Camera Offset Amount=-5) {module=CameraOffset001}
   &SpriteRotationRate(Rotation Rate=&RandomRangeFloat(Minimum=-50, Maximum=50))
  ]
  rendererClasses [&NiagaraSpriteRendererProperties]
  renderers [{rendererClass=&NiagaraSpriteRendererProperties}]
 }
 Wheel_Sprites_BackExtra {
  emitterUpdateScript [
   &EmitterState()
   &SpawnRate(SpawnRate=20, Use Spawn Probability=false)
  ]
  particleSpawnScript [
   &InitializeParticle(
    Color=<rgba 5 3.0056 1.6206 0.75>
    Lifetime=&UniformRangedFloat(Minimum=0.48, Maximum=0.9)
    Module.Write Color=false
    Module.Write Lifetime=true
    Module.Write Mass=false
    Module.Write Material Random=true
    Module.Write Position=false
    Module.Write Scale=false
    Module.Write SpriteRotation=true
    Module.Write SpriteSize=true
    Position=&SimulationPosition()
    Sprite Rotation=&UniformRangedFloat(Minimum=-20, Maximum=20)
    Sprite Size=&MakeCustomVector2DFromBool(Bool=&RandomBool(), True Vector2D=<35 35>, False Vector2D=<-35 -35>)
   )
   &FN_SetSpawnGroup_v2(Number of Spawn Groups=2)
   &SkeletalMeshLocation(
    Apply To Particle Position=false
    Bone / Socket Index=&SpawnGroup()
    Bone Sampling Mode=::NewEnumerator4
    Inherit Velocity=false
    Interpolate Mesh Sampling=false
    MaskSpawnGroup=false
    Skeletal Mesh=$User.SkeletalMesh
   )
   &CylinderLocation(
    Cylinder Height=0
    Cylinder Radius=25
    Hemisphere Y=true
    Lathe Profile=DataInterface() {propertyValues {Curve {keys [<at 0 0.451 RCIM_Cubic>, {arriveTangent=0.049, interpMode=RCIM_Cubic, leaveTangent=0.049, time=0.5, value=1}, <at 1 0.5 RCIM_Cubic>]}}}
    MaskSpawnGroup=false
    Module.Enable Lathe Profile=false
    Module.Override Randomness=false
    Module.Surface Only=true
    Offset=&MaskVectorBySpawnGroup(Masked Vector=<-11 0 5>, Passthrough Vector=<11 0 5>)
    Orientation Axis:ENiagaraOrientationAxis=::XAxis
    Surface Only Band Thickness=0
    Use Endcaps In Surface Only Mode=false
   ) {module=CylinderLocation001}
   /Niagara/Modules/Spawn/Location/SphereLocation(MaskSpawnGroup=false, Module.Override Randomness=false, Non Uniform Scale=<1 1 0.1>, Sphere Radius=40, Surface Only=false) {enabled=false}
   &DynamicMaterialParameters(
    Index 0 Param 2=&UniformRangedFloat(Minimum=0, Maximum=1)
    Index 0 Param 3=1
    Index 0 Param 4=&UniformRangedFloat(Minimum=1, Maximum=2)
    Module.Param0WriteEnabled=false
    Module.Param1WriteEnabled=true
    Module.Param2WriteEnabled=false
    Module.Param3WriteEnabled=true
   ) {module=DynamicMaterialParameters002}
   /SetVariables(Particles.CylVector:Vector3f=$Output.CylinderLocation001.CylinderVector) {isSetParameters=true, module=SetVariables_#2}
   /SetVariables(Particles.DestinationVector:Vector3f=&Add_Vector(A=$Output.SkeletalMeshLocation.SampledPosition, B=$Particles.CylVector)) {isSetParameters=true, module=SetVariables_#3}
  ]
  particleUpdateScript [
   &ParticleState()
   &SkeletalMeshLocation(
    Apply To Particle Position=false
    Bone / Socket Index=1
    Bone Sampling Mode=::NewEnumerator4
    Inherit Velocity=false
    Interpolate Mesh Sampling=false
    MaskSpawnGroup=false
    Skeletal Mesh=$User.SkeletalMesh
   ) {module=SkeletalMeshLocation001}
   &self:UseUpdatedSkelLocation(
    DestinationPosition:Vector3f=&Add_Vector(A=$Particles.CylVector, B=$Output.SkeletalMeshLocation001.SampledPosition)
    Normalized Influence:NiagaraFloat=&FloatFromCurve(FloatCurve=DataInterface() {propertyValues {Curve {keys [<at 0 1>,<at 1 0.91>]}}})
    Offset Position:Vector3f=&MaskVectorBySpawnGroup(
     Masked Vector=&RandomRangeVector(
      Minimum=&VectorFromCurve(VectorCurve=DataInterface() {propertyValues {
         XCurve {keys [{arriveTangent=5.5533, time=0, value=0}, {arriveTangent=65.1046, leaveTangent=65.1043, time=1, value=-10}]}
         YCurve {keys [<at 0 -2>]}
         ZCurve {keys [<at 0 0 RCIM_Cubic>, {arriveTangent=68.343, interpMode=RCIM_Cubic, leaveTangent=68.3433, tangentMode=RCTM_User, time=1, value=22}]}
        }})
      Maximum=&VectorFromCurve(VectorCurve=DataInterface() {propertyValues {
         XCurve {keys [{arriveTangent=-0.158, time=0, value=0}, {arriveTangent=55.3024, leaveTangent=55.3025, time=1, value=-20}]}
         YCurve {keys [<at 0 3>]}
         ZCurve {keys [<at 0 0 RCIM_Cubic>, {arriveTangent=83.3425, interpMode=RCIM_Cubic, leaveTangent=83.3426, tangentMode=RCTM_User, time=1, value=22}]}
        }})
     )
     Passthrough Vector=&RandomRangeVector(
      Minimum=&VectorFromCurve(VectorCurve=DataInterface() {propertyValues {
         XCurve {keys [{arriveTangent=5.5533, time=0, value=0}, {arriveTangent=65.1046, leaveTangent=65.1043, time=1, value=10}]}
         YCurve {keys [<at 0 -2>]}
         ZCurve {keys [<at 0 0 RCIM_Cubic>, {arriveTangent=68.343, interpMode=RCIM_Cubic, leaveTangent=68.3433, tangentMode=RCTM_User, time=1, value=22}]}
        }})
      Maximum=&VectorFromCurve(VectorCurve=DataInterface() {propertyValues {
         XCurve {keys [{arriveTangent=-0.158, time=0, value=0}, {arriveTangent=55.3024, leaveTangent=55.3025, time=1, value=20}]}
         YCurve {keys [<at 0 3>]}
         ZCurve {keys [<at 0 0 RCIM_Cubic>, {arriveTangent=83.3425, interpMode=RCIM_Cubic, leaveTangent=83.3426, tangentMode=RCTM_User, time=1, value=22}]}
        }})
     )
    )
    SourcePosition:Vector3f=$Particles.DestinationVector
   ) {module=UseUpdatedSkelLocation}
   &ScaleColor(
    Scale Mode=::NewEnumerator1
    Scale RGBA=&Vector4FromCurve(Vector4Curve=DataInterface() {propertyValues {
       AlphaCurve {keys [<at 0.0024 0>,<at 0.1271 1.2>,<at 0.75 4>]}
       BlueCurve {keys [<at 0.009 0.15>,<at 0.25 0>,<at 0.6 0>,<at 0.75 0>]}
       GreenCurve {keys [<at 0.009 0.5841>,<at 0.25 0.125>,<at 0.6 0.0833>,<at 0.75 0>]}
       RedCurve {keys [<at 0.009 1>,<at 0.25 1>,<at 0.6 1>,<at 0.75 1>]}
      }})
    ScaleRGBA=true
   ) {module=ScaleColor002}
   /SetVariables(Particles.SubImageIndex=&Add_Float(
     A=&RandomRangeFloat(Minimum=0, Maximum=20)
     B=&FloatFromCurve(FloatCurve=DataInterface() {propertyValues {Curve {keys [<at 0 0>,<at 1 24>]}}})
    )) {isSetParameters=true, module=SetVariables_#4}
   &ScaleSpriteSize(Scale Factor=&Vector2DFromFloat(Value=&FloatFromCurve(FloatCurve=DataInterface() {propertyValues {Curve {keys [
          {arriveTangent=7.5764, interpMode=RCIM_Cubic, leaveTangent=7.5765, tangentMode=RCTM_User, time=0, value=0}
          {arriveTangent=-0.0353, interpMode=RCIM_Cubic, leaveTangent=-0.0353, tangentMode=RCTM_User, time=0.2, value=1}
          <at 0.5752 1>
          <at 1 0 RCIM_Cubic>
         ]}}})))
   &DynamicMaterialParameters(
    Index 0 Param 1=&FloatFromCurve(FloatCurve=DataInterface() {propertyValues {Curve {keys [{arriveTangent=2.7551, interpMode=RCIM_Cubic, leaveTangent=2.7551, tangentMode=RCTM_User, time=0, value=0}, <at 0.5 1>]}}})
    Index 0 Param 2=&RandomRangeFloat(Minimum=0.33, Maximum=1)
    Index 0 Param 3=&FloatFromCurve(FloatCurve=DataInterface() {propertyValues {Curve {keys [<at 0 1>,<at 1 2>]}}})
    Index 0 Param 4=1
    Module.Param0WriteEnabled=true
    Module.Param1WriteEnabled=true
    Module.Param2WriteEnabled=true
    Module.Param3WriteEnabled=true
   ) {module=DynamicMaterialParameters001}
   &CameraOffset() {module=CameraOffset001}
   &SpriteRotationRate(Rotation Rate=&RandomRangeFloat(Minimum=-50, Maximum=50))
  ]
  rendererClasses [&NiagaraSpriteRendererProperties]
  renderers [{rendererClass=&NiagaraSpriteRendererProperties}]
 }
 Wheel_Sprites_Flamelicks {
  emitterUpdateScript [
   &EmitterState()
   &SpawnRate(SpawnRate=10, Use Spawn Probability=false)
  ]
  particleSpawnScript [
   &InitializeParticle(
    Color=<rgba 5 3.0056 1.6206 0.75>
    Lifetime=&UniformRangedFloat(Minimum=0.6, Maximum=1.1)
    Module.Write Color=false
    Module.Write Lifetime=true
    Module.Write Mass=false
    Module.Write Material Random=true
    Module.Write Position=false
    Module.Write Scale=false
    Module.Write SpriteRotation=true
    Module.Write SpriteSize=true
    Position=&SimulationPosition()
    Sprite Rotation=&UniformRangedFloat(Minimum=-20, Maximum=20)
    Sprite Size=&MakeCustomVector2DFromBool(Bool=&RandomBool(), True Vector2D=<45 45>, False Vector2D=<-45 -45>)
   )
   &FN_SetSpawnGroup_v2(Number of Spawn Groups=2)
   &SkeletalMeshLocation(
    Apply To Particle Position=false
    Bone Sampling Mode=::NewEnumerator4
    Inherit Velocity=false
    Interpolate Mesh Sampling=false
    MaskSpawnGroup=false
    Skeletal Mesh=$User.SkeletalMesh
   )
   &CylinderLocation(
    Cylinder Height=7.5
    Cylinder Radius=25
    Lathe Profile=DataInterface() {propertyValues {Curve {keys [<at 0 0.451 RCIM_Cubic>, {arriveTangent=0.049, interpMode=RCIM_Cubic, leaveTangent=0.049, time=0.5, value=1}, <at 1 0.5 RCIM_Cubic>]}}}
    MaskSpawnGroup=false
    Module.Enable Lathe Profile=false
    Module.Override Randomness=false
    Module.Surface Only=true
    Offset=<0 0 5>
    Orientation Axis:ENiagaraOrientationAxis=::XAxis
    Surface Only Band Thickness=0
    Use Endcaps In Surface Only Mode=false
   ) {module=CylinderLocation001}
   &SphereLocation(MaskSpawnGroup=false, Module.Override Randomness=false, Non Uniform Scale=<1 1 0.1>, Sphere Radius=40, Surface Only=false) {enabled=false}
   &DynamicMaterialParameters(
    Index 0 Param 2=&UniformRangedFloat(Minimum=0, Maximum=1)
    Index 0 Param 3=1
    Index 0 Param 4=&UniformRangedFloat(Minimum=1, Maximum=2)
    Module.Param0WriteEnabled=false
    Module.Param1WriteEnabled=true
    Module.Param2WriteEnabled=false
    Module.Param3WriteEnabled=true
   ) {module=DynamicMaterialParameters002}
   /SetVariables(Particles.CylVector:Vector3f=$Output.CylinderLocation001.CylinderVector) {isSetParameters=true, module=SetVariables_#2}
   /SetVariables(Particles.DestinationVector:Vector3f=&Add_Vector(A=$Output.SkeletalMeshLocation.SampledPosition, B=$Particles.CylVector)) {isSetParameters=true, module=SetVariables_#3}
  ]
  particleUpdateScript [
   &ParticleState()
   &SkeletalMeshLocation(
    Apply To Particle Position=false
    Bone Sampling Mode=::NewEnumerator4
    Inherit Velocity=false
    Interpolate Mesh Sampling=false
    MaskSpawnGroup=false
    Skeletal Mesh=$User.SkeletalMesh
   ) {module=SkeletalMeshLocation001}
   &self:UseUpdatedSkelLocation(
    DestinationPosition:Vector3f=&Add_Vector(A=$Particles.CylVector, B=$Output.SkeletalMeshLocation001.SampledPosition)
    Normalized Influence:NiagaraFloat=&FloatFromCurve(FloatCurve=DataInterface() {propertyValues {Curve {keys [<at 0 1>,<at 1 0.91>]}}})
    Offset Position:Vector3f=&RandomRangeVector(
     Minimum=&VectorFromCurve(VectorCurve=DataInterface() {propertyValues {
        XCurve {keys [<at 0 -2>]}
        YCurve {keys [<at 0 -2>]}
        ZCurve {keys [<at 0 0 RCIM_Cubic>, {arriveTangent=29.403, interpMode=RCIM_Cubic, leaveTangent=29.4031, tangentMode=RCTM_User, time=1, value=40}]}
       }})
     Maximum=&VectorFromCurve(VectorCurve=DataInterface() {propertyValues {
        XCurve {keys [<at 0 3>]}
        YCurve {keys [<at 0 3>]}
        ZCurve {keys [<at 0 0 RCIM_Cubic>, {arriveTangent=29.403, interpMode=RCIM_Cubic, leaveTangent=29.4031, tangentMode=RCTM_User, time=1, value=40}]}
       }})
    )
    SourcePosition:Vector3f=$Particles.DestinationVector
   ) {module=UseUpdatedSkelLocation}
   &ScaleColor(
    Scale Mode=::NewEnumerator1
    Scale RGBA=&Vector4FromCurve(Vector4Curve=DataInterface() {propertyValues {
       AlphaCurve {keys [<at 0.0024 0>,<at 0.1271 1.2>,<at 0.75 4>]}
       BlueCurve {keys [<at 0.009 0.15>,<at 0.25 0>,<at 0.6 0>,<at 0.75 0>]}
       GreenCurve {keys [<at 0.009 0.5841>,<at 0.25 0.125>,<at 0.6 0.0833>,<at 0.75 0>]}
       RedCurve {keys [<at 0.009 1>,<at 0.25 1>,<at 0.6 1>,<at 0.75 1>]}
      }})
    ScaleRGBA=true
   ) {module=ScaleColor002}
   /SetVariables(Particles.SubImageIndex=&Add_Float(
     A=&RandomRangeFloat(Minimum=0, Maximum=20)
     B=&FloatFromCurve(FloatCurve=DataInterface() {propertyValues {Curve {keys [<at 0 0>,<at 1 24>]}}})
    )) {isSetParameters=true, module=SetVariables_#4}
   &ScaleSpriteSize(Scale Factor=&Vector2DFromFloat(Value=&FloatFromCurve(FloatCurve=DataInterface() {propertyValues {Curve {keys [
          {arriveTangent=7.5764, interpMode=RCIM_Cubic, leaveTangent=7.5765, tangentMode=RCTM_User, time=0, value=0}
          {arriveTangent=-0.0353, interpMode=RCIM_Cubic, leaveTangent=-0.0353, tangentMode=RCTM_User, time=0.2, value=1}
          <at 0.8 1>
          <at 1 0 RCIM_Cubic>
         ]}}})))
   &DynamicMaterialParameters(
    Index 0 Param 1=&FloatFromCurve(FloatCurve=DataInterface() {propertyValues {Curve {keys [{arriveTangent=2.7551, interpMode=RCIM_Cubic, leaveTangent=2.7551, tangentMode=RCTM_User, time=0, value=0}, <at 0.5 1>]}}})
    Index 0 Param 2=&RandomRangeFloat(Minimum=0.33, Maximum=1)
    Index 0 Param 3=&FloatFromCurve(FloatCurve=DataInterface() {propertyValues {Curve {keys [<at 0 1>,<at 1 2>]}}})
    Index 0 Param 4=1
    Module.Param0WriteEnabled=true
    Module.Param1WriteEnabled=true
    Module.Param2WriteEnabled=true
    Module.Param3WriteEnabled=true
   ) {module=DynamicMaterialParameters001}
   &CameraOffset(Camera Offset Amount=-20) {module=CameraOffset001}
   &SpriteRotationRate(Rotation Rate=&RandomRangeFloat(Minimum=-50, Maximum=50))
  ]
  rendererClasses [&NiagaraSpriteRendererProperties]
  renderers [{rendererClass=&NiagaraSpriteRendererProperties}]
 }
 Wheel_Sprites_Front {
  emitterUpdateScript [
   &EmitterState()
   &SpawnRate(SpawnRate=30, Use Spawn Probability=false)
  ]
  particleSpawnScript [
   &InitializeParticle(
    Color=<rgba 5 3.0056 1.6206 0.75>
    Lifetime=&UniformRangedFloat(Minimum=0.48, Maximum=0.9)
    Module.Write Color=false
    Module.Write Lifetime=true
    Module.Write Mass=false
    Module.Write Material Random=true
    Module.Write Position=false
    Module.Write Scale=false
    Module.Write SpriteRotation=true
    Module.Write SpriteSize=true
    Position=&SimulationPosition()
    Sprite Rotation=&UniformRangedFloat(Minimum=-20, Maximum=20)
    Sprite Size=&MakeCustomVector2DFromBool(Bool=&RandomBool(), True Vector2D=<52 52>, False Vector2D=<-52 52>)
   )
   &FN_SetSpawnGroup_v2(Number of Spawn Groups=2)
   &SkeletalMeshLocation(
    Apply To Particle Position=false
    Bone / Socket Index=&SpawnGroup()
    Bone Sampling Mode=::NewEnumerator4
    Inherit Velocity=false
    Interpolate Mesh Sampling=false
    MaskSpawnGroup=false
    Skeletal Mesh=$User.SkeletalMesh
   )
   &CylinderLocation(
    Cylinder Height=7.5
    Cylinder Radius=25
    Lathe Profile=DataInterface() {propertyValues {Curve {keys [<at 0 0.451 RCIM_Cubic>, {arriveTangent=0.049, interpMode=RCIM_Cubic, leaveTangent=0.049, time=0.5, value=1}, <at 1 0.5 RCIM_Cubic>]}}}
    MaskSpawnGroup=false
    Module.Enable Lathe Profile=false
    Module.Override Randomness=false
    Module.Surface Only=true
    Offset=<0 0 5>
    Orientation Axis:ENiagaraOrientationAxis=::XAxis
    Surface Only Band Thickness=0
    Use Endcaps In Surface Only Mode=false
   ) {module=CylinderLocation001}
   &SphereLocation(MaskSpawnGroup=false, Module.Override Randomness=false, Non Uniform Scale=<1 1 0.1>, Sphere Radius=40, Surface Only=false) {enabled=false}
   &DynamicMaterialParameters(
    Index 0 Param 2=&UniformRangedFloat(Minimum=0, Maximum=1)
    Index 0 Param 3=1
    Index 0 Param 4=&UniformRangedFloat(Minimum=1, Maximum=2)
    Module.Param0WriteEnabled=false
    Module.Param1WriteEnabled=true
    Module.Param2WriteEnabled=false
    Module.Param3WriteEnabled=true
   ) {module=DynamicMaterialParameters002}
   /SetVariables(Particles.CylVector:Vector3f=$Output.CylinderLocation001.CylinderVector) {isSetParameters=true, module=SetVariables_#2}
   /SetVariables(Particles.DestinationVector:Vector3f=&Add_Vector(A=$Output.SkeletalMeshLocation.SampledPosition, B=$Particles.CylVector)) {isSetParameters=true, module=SetVariables_#3}
  ]
  particleUpdateScript [
   &ParticleState()
   &SkeletalMeshLocation(
    Apply To Particle Position=false
    Bone Sampling Mode=::NewEnumerator4
    Inherit Velocity=false
    Interpolate Mesh Sampling=false
    MaskSpawnGroup=false
    Skeletal Mesh=$User.SkeletalMesh
   ) {module=SkeletalMeshLocation001}
   &self:UseUpdatedSkelLocation(
    DestinationPosition:Vector3f=&Add_Vector(A=$Particles.CylVector, B=$Output.SkeletalMeshLocation001.SampledPosition)
    Normalized Influence:NiagaraFloat=&FloatFromCurve(FloatCurve=DataInterface() {propertyValues {Curve {keys [<at 0 1>,<at 1 0.91>]}}})
    Offset Position:Vector3f=&RandomRangeVector(
     Minimum=&VectorFromCurve(VectorCurve=DataInterface() {propertyValues {
        XCurve {keys [<at 0 -2>]}
        YCurve {keys [<at 0 -2>]}
        ZCurve {keys [<at 0 0 RCIM_Cubic>, {arriveTangent=29.403, interpMode=RCIM_Cubic, leaveTangent=29.4031, tangentMode=RCTM_User, time=1, value=22}]}
       }})
     Maximum=&VectorFromCurve(VectorCurve=DataInterface() {propertyValues {
        XCurve {keys [<at 0 3>]}
        YCurve {keys [<at 0 3>]}
        ZCurve {keys [<at 0 0 RCIM_Cubic>, {arriveTangent=29.403, interpMode=RCIM_Cubic, leaveTangent=29.4031, tangentMode=RCTM_User, time=1, value=22}]}
       }})
    )
    SourcePosition:Vector3f=$Particles.DestinationVector
   ) {module=UseUpdatedSkelLocation}
   &ScaleColor(
    Scale Mode=::NewEnumerator1
    Scale RGBA=&Vector4FromCurve(Vector4Curve=DataInterface() {propertyValues {
       AlphaCurve {keys [<at 0.0024 0>,<at 0.1271 1.2>,<at 0.75 4>]}
       BlueCurve {keys [<at 0.009 0.15>,<at 0.25 0>,<at 0.6 0>,<at 0.75 0>]}
       GreenCurve {keys [<at 0.009 0.5841>,<at 0.25 0.125>,<at 0.6 0.0833>,<at 0.75 0>]}
       RedCurve {keys [<at 0.009 1>,<at 0.25 1>,<at 0.6 1>,<at 0.75 1>]}
      }})
    ScaleRGBA=true
   ) {module=ScaleColor002}
   /SetVariables(Particles.SubImageIndex=&Add_Float(
     A=&RandomRangeFloat(Minimum=0, Maximum=20)
     B=&FloatFromCurve(FloatCurve=DataInterface() {propertyValues {Curve {keys [<at 0 0>,<at 1 24>]}}})
    )) {isSetParameters=true, module=SetVariables_#4}
   &ScaleSpriteSize(Scale Factor=&Vector2DFromFloat(Value=&FloatFromCurve(FloatCurve=DataInterface() {propertyValues {Curve {keys [
          {arriveTangent=7.5764, interpMode=RCIM_Cubic, leaveTangent=7.5765, tangentMode=RCTM_User, time=0, value=0}
          {arriveTangent=-0.0353, interpMode=RCIM_Cubic, leaveTangent=-0.0353, tangentMode=RCTM_User, time=0.2, value=1}
          <at 0.5752 1>
          <at 1 0 RCIM_Cubic>
         ]}}})))
   &DynamicMaterialParameters(
    Index 0 Param 1=&FloatFromCurve(FloatCurve=DataInterface() {propertyValues {Curve {keys [{arriveTangent=2.7551, interpMode=RCIM_Cubic, leaveTangent=2.7551, tangentMode=RCTM_User, time=0, value=0}, <at 0.5 1>]}}})
    Index 0 Param 2=&RandomRangeFloat(Minimum=0.33, Maximum=1)
    Index 0 Param 3=&FloatFromCurve(FloatCurve=DataInterface() {propertyValues {Curve {keys [<at 0 1>,<at 1 2>]}}})
    Index 0 Param 4=1
    Module.Param0WriteEnabled=true
    Module.Param1WriteEnabled=true
    Module.Param2WriteEnabled=true
    Module.Param3WriteEnabled=true
   ) {module=DynamicMaterialParameters001}
   &CameraOffset(Camera Offset Amount=-5) {module=CameraOffset001}
   &SpriteRotationRate(Rotation Rate=&RandomRangeFloat(Minimum=-50, Maximum=50))
  ]
  rendererClasses [&NiagaraSpriteRendererProperties]
  renderers [{rendererClass=&NiagaraSpriteRendererProperties}]
 }
}
summary {
 systemName=NS_Embers_Glider_Idle
 emitters [
  {emitterName=Wheel_Sprites_Front, rendererClasses [&NiagaraSpriteRendererProperties]}
  {emitterName=WheelFire_Front, rendererClasses [&NiagaraMeshRendererProperties]}
  {emitterName=WheelFire_Back, rendererClasses [&NiagaraMeshRendererProperties]}
  {emitterName=WheelFire_Back_Bottom, rendererClasses [&NiagaraMeshRendererProperties]}
  {emitterName=WheelFire_Front_Bottom, rendererClasses [&NiagaraMeshRendererProperties]}
  {emitterName=WheelFire_ForwardFacing, rendererClasses [&NiagaraMeshRendererProperties]}
  {emitterName=Handlebar_Fire, rendererClasses [&NiagaraMeshRendererProperties]}
  {emitterName=Frame_FIre, rendererClasses [&NiagaraMeshRendererProperties]}
  {emitterName=Eyefire, rendererClasses [&NiagaraMeshRendererProperties]}
  {emitterName=Wheel_Sprites_Flamelicks, rendererClasses [&NiagaraSpriteRendererProperties]}
  {emitterName=Wheel_Sprites_BackExtra, rendererClasses [&NiagaraSpriteRendererProperties]}
  {emitterName=Light, rendererClasses [&NiagaraLightRendererProperties]}
  {emitterName=Wheel_Sprites_Back, rendererClasses [&NiagaraSpriteRendererProperties]}
 ]
 userVariables [
  User.bIsFrontEndPreview {type=NiagaraBool}
  User.ForwardDot {type=NiagaraFloat, defaultValue=0, description=""}
  User.RightDot {type=NiagaraFloat, defaultValue=0, description=""}
  User.SkeletalMesh {
   type=NiagaraDataInterfaceSkeletalMesh
   defaultValue {
    struct=NiagaraExt_VariableValue_DataInterface
    value {dataInterface=&self:NiagaraDataInterfaceSkeletalMesh_0}
   }
  }
 ]
}
